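# Physics asset for the York Male pickaxe skeletal mesh, a two-handed melee cosmetic with jointed wire pieces.
# A long capsule covers the handle bone, and five small capsules cover the dyn_Wire bones.
# Wires 01 to 04 are constrained directly to the handle, while Wire 05 chains off Wire 04.
# Tight angular limits restrict Wires 01 to 03 mostly to one swing axis, while 04 and 05 flex evenly in swing and twist.
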PhysicsAsset /BRCosmetics/Weapons/FORT_Melee/Pickaxe_York_Male/Meshes/SK_York_Male_Pickaxe_PhysicsAsset
bodies {
 Melee_TwoHanded_Handle [{shapeType=Capsule, length=118.4636, radius=3.3998, center=<0.5972 -0.3017 23.9683>, rotation=<rot 1.0747 -89.9998 -0.6088>}]
 dyn_Wire_01 [{shapeType=Capsule, length=4.8415, radius=0.7729, center=<-0.391 1.6511 0.1431>, rotation=<rot 83.3417 -9.0497 82.9292>}]
 dyn_Wire_02 [{shapeType=Capsule, length=1.0144, radius=2.118, center=<0.0103 2.3504 -0.0791>, rotation=<rot 85.327 174.3993 -98.1034>}]
 dyn_Wire_03 [{shapeType=Capsule, length=2.2654, radius=2.2506, center=<0.106 -0.0025 -2.8363>, rotation=<rot 0 -180 -3.4359>}]
 dyn_Wire_04 [{shapeType=Capsule, length=3.7901, radius=0.6201, center=<-0.063 0.0182 -1.6897>, rotation=<rot 0 -90 0>}]
 dyn_Wire_05 [{shapeType=Capsule, length=2.7181, radius=0.6417, center=<-0.063 0.0458 -0.9755>, rotation=<rot 1.6318 -90 0>}]
}
constraints [
 {bone1Name=dyn_Wire_01, bone2Name=Melee_TwoHanded_Handle, swing1LimitDegrees=20, swing2LimitDegrees=2, twistLimitDegrees=2}
 {bone1Name=dyn_Wire_02, bone2Name=Melee_TwoHanded_Handle, swing1LimitDegrees=40, swing2LimitDegrees=2, twistLimitDegrees=2}
 {bone1Name=dyn_Wire_03, bone2Name=Melee_TwoHanded_Handle, swing1LimitDegrees=2, swing2LimitDegrees=40, twistLimitDegrees=2}
 {bone1Name=dyn_Wire_04, bone2Name=Melee_TwoHanded_Handle, swing1LimitDegrees=20, swing2LimitDegrees=20, twistLimitDegrees=20}
 {bone1Name=dyn_Wire_05, bone2Name=dyn_Wire_04, swing1LimitDegrees=15, swing2LimitDegrees=15, twistLimitDegrees=15}
]
